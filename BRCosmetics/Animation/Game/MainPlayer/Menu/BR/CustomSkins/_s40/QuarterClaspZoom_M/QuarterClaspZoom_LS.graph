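# Level Sequence for the QuarterClaspZoom_M custom-skin menu animation in Battle Royale, animating a spawned FortPlayerMannequin.
# A skeletal animation section supplies the base motion, with the mannequin's transform at identity.
# A layered Fortnite_Biped_CtrlRig track adds corrective offsets on top.
# These include IK legs with stretch and softness, near-static foot adjustments and brief keyed rotation fixes on the left forearm and hand.
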LevelSequence /BRCosmetics/Animation/Game/MainPlayer/Menu/BR/CustomSkins/_s40/QuarterClaspZoom_M/QuarterClaspZoom_LS
bindings [FortPlayerMannequin {
  id=#1
  possessed_class=FortPlayerMannequin
  tracks [
   {type=MovieSceneSpawnTrack, sections [{type=MovieSceneSpawnSection, channels [None {type=MovieSceneScriptingBoolChannel, default=true}]}]}
   {
    type=MovieScene3DTransformTrack
    sections [{
      type=MovieScene3DTransformSection
      channels [
       Location.X {type=MovieSceneScriptingDoubleChannel, default=0}
       Location.Y {type=MovieSceneScriptingDoubleChannel, default=0}
       Location.Z {type=MovieSceneScriptingDoubleChannel, default=0}
       Rotation.X {type=MovieSceneScriptingDoubleChannel, default=0}
       Rotation.Y {type=MovieSceneScriptingDoubleChannel, default=0}
       Rotation.Z {type=MovieSceneScriptingDoubleChannel, default=0}
       Scale.X {type=MovieSceneScriptingDoubleChannel, default=1}
       Scale.Y {type=MovieSceneScriptingDoubleChannel, default=1}
       Scale.Z {type=MovieSceneScriptingDoubleChannel, default=1}
      ]
     }]
   }
   {type=MovieSceneSkeletalAnimationTrack, sections [{type=MovieSceneSkeletalAnimationSection, end_frame=290, start_frame=0}]}
   {
    type=MovieSceneControlRigParameterTrack
    display_name="Fortnite_Biped_CtrlRig (Layered)"
    sections [{
      type=MovieSceneControlRigParameterSection
      channels [
       leg_l_stretch_switch {type=MovieSceneScriptingBoolChannel, default=true, const [true,2]}
       leg_r_stretch_switch {type=MovieSceneScriptingBoolChannel, default=true, const [true,2]}
       arm_l_fk_local {type=MovieSceneScriptingBoolChannel, default=true}
       Auto Interpolate_3 {type=MovieSceneScriptingBoolChannel, default=false}
       Auto Interpolate {type=MovieSceneScriptingBoolChannel, default=false}
       arm_l_stretch_switch {type=MovieSceneScriptingBoolChannel, default=false}
       Keep Aligned {type=MovieSceneScriptingBoolChannel, default=false}
       arm_r_fk_local {type=MovieSceneScriptingBoolChannel, default=true}
       Auto Interpolate_4 {type=MovieSceneScriptingBoolChannel, default=false}
       Auto Interpolate_2 {type=MovieSceneScriptingBoolChannel, default=false}
       arm_r_stretch_switch {type=MovieSceneScriptingBoolChannel, default=false}
       Keep Aligned_2 {type=MovieSceneScriptingBoolChannel, default=false}
       head_local {type=MovieSceneScriptingBoolChannel, default=true}
       Stretch {type=MovieSceneScriptingBoolChannel, default=false}
       head_stretch {type=MovieSceneScriptingBoolChannel, default=false}
       spine_stretch_switch {type=MovieSceneScriptingBoolChannel, default=false}
       arm_l_fk_ik_switch {type=MovieSceneScriptingBoolChannel, default=false, const [false,1]}
       leg_l_fk_ik_switch {type=MovieSceneScriptingBoolChannel, default=true, const [true,1]}
       arm_r_fk_ik_switch {type=MovieSceneScriptingBoolChannel, default=false, const [false,2]}
       leg_r_fk_ik_switch {type=MovieSceneScriptingBoolChannel, default=false, const [true,1]}
       ShowBodyControls {type=MovieSceneScriptingBoolChannel, default=true, const [true,1]}
       spine_fk_ik_switch {type=MovieSceneScriptingBoolChannel, default=false, const [false,1]}
       neck_fk_ik_switch {type=MovieSceneScriptingBoolChannel, default=false, const [false,1]}
       space {type=MovieSceneScriptingByteChannel, default=1}
       global_ctrl.Location.X {type=MovieSceneScriptingFloatChannel, default=0, const [0,1]}
       global_ctrl.Location.Y {type=MovieSceneScriptingFloatChannel, default=0, const [0,1]}
       global_ctrl.Location.Z {type=MovieSceneScriptingFloatChannel, default=0, const [0,1]}
       global_ctrl.Rotation.X {type=MovieSceneScriptingFloatChannel, default=0, const [0,1]}
       global_ctrl.Rotation.Y {type=MovieSceneScriptingFloatChannel, default=0, const [0,1]}
       global_ctrl.Rotation.Z {type=MovieSceneScriptingFloatChannel, default=0, const [0,1]}
       root_ctrl.Location.X {type=MovieSceneScriptingFloatChannel, default=0}
       root_ctrl.Location.Y {type=MovieSceneScriptingFloatChannel, default=0}
       root_ctrl.Location.Z {type=MovieSceneScriptingFloatChannel, default=0}
       root_ctrl.Rotation.X {type=MovieSceneScriptingFloatChannel, default=0}
       root_ctrl.Rotation.Y {type=MovieSceneScriptingFloatChannel, default=0}
       root_ctrl.Rotation.Z {type=MovieSceneScriptingFloatChannel, default=0}
       root_ctrl.Scale.X {type=MovieSceneScriptingFloatChannel, default=1}
       root_ctrl.Scale.Y {type=MovieSceneScriptingFloatChannel, default=1}
       root_ctrl.Scale.Z {type=MovieSceneScriptingFloatChannel, default=1}
       body_offset_ctrl.Location.X {type=MovieSceneScriptingFloatChannel, default=0}
       body_offset_ctrl.Location.Y {type=MovieSceneScriptingFloatChannel, default=0}
       body_offset_ctrl.Location.Z {type=MovieSceneScriptingFloatChannel, default=0}
       body_offset_ctrl.Rotation.X {type=MovieSceneScriptingFloatChannel, default=0}
       body_offset_ctrl.Rotation.Y {type=MovieSceneScriptingFloatChannel, default=0}
       body_offset_ctrl.Rotation.Z {type=MovieSceneScriptingFloatChannel, default=0}
       body_offset_ctrl.Scale.X {type=MovieSceneScriptingFloatChannel, default=1}
       body_offset_ctrl.Scale.Y {type=MovieSceneScriptingFloatChannel, default=1}
       body_offset_ctrl.Scale.Z {type=MovieSceneScriptingFloatChannel, default=1}
       thigh_l_fk_ctrl.Location.X {type=MovieSceneScriptingFloatChannel, default=0, const [0,1]}
       thigh_l_fk_ctrl.Location.Y {type=MovieSceneScriptingFloatChannel, default=0, const [0,1]}
       thigh_l_fk_ctrl.Location.Z {type=MovieSceneScriptingFloatChannel, default=0, const [0,1]}
       thigh_l_fk_ctrl.Rotation.X {type=MovieSceneScriptingFloatChannel, default=-0.0445, const [-0.0909,1]}
       thigh_l_fk_ctrl.Rotation.Y {type=MovieSceneScriptingFloatChannel, default=-0.0019, const [-0.0019,1]}
       thigh_l_fk_ctrl.Rotation.Z {type=MovieSceneScriptingFloatChannel, default=0, const [0,1]}
       thigh_l_fk_ctrl.Scale.X {type=MovieSceneScriptingFloatChannel, default=1, const [1,1]}
       thigh_l_fk_ctrl.Scale.Y {type=MovieSceneScriptingFloatChannel, default=1, const [1,1]}
       thigh_l_fk_ctrl.Scale.Z {type=MovieSceneScriptingFloatChannel, default=1, const [1,1]}
       calf_l_fk_ctrl.Location.X {type=MovieSceneScriptingFloatChannel, default=0, const [0,1]}
       calf_l_fk_ctrl.Location.Y {type=MovieSceneScriptingFloatChannel, default=0, const [0,1]}
       calf_l_fk_ctrl.Location.Z {type=MovieSceneScriptingFloatChannel, default=0, const [0,1]}
       calf_l_fk_ctrl.Rotation.X {type=MovieSceneScriptingFloatChannel, default=0, const [0,1]}
       calf_l_fk_ctrl.Rotation.Y {type=MovieSceneScriptingFloatChannel, default=0, const [0,1]}
       calf_l_fk_ctrl.Rotation.Z {type=MovieSceneScriptingFloatChannel, default=0, const [0,1]}
       calf_l_fk_ctrl.Scale.X {type=MovieSceneScriptingFloatChannel, default=1, const [1,1]}
       calf_l_fk_ctrl.Scale.Y {type=MovieSceneScriptingFloatChannel, default=1, const [1,1]}
       calf_l_fk_ctrl.Scale.Z {type=MovieSceneScriptingFloatChannel, default=1, const [1,1]}
       foot_l_fk_ctrl.Location.X {type=MovieSceneScriptingFloatChannel, default=0, const [0,1]}
       foot_l_fk_ctrl.Location.Y {type=MovieSceneScriptingFloatChannel, default=0, const [0,1]}
       foot_l_fk_ctrl.Location.Z {type=MovieSceneScriptingFloatChannel, default=0, const [0,1]}
       foot_l_fk_ctrl.Rotation.X {type=MovieSceneScriptingFloatChannel, default=0.0445, const [0.0909,1]}
       foot_l_fk_ctrl.Rotation.Y {type=MovieSceneScriptingFloatChannel, default=-0.0016, const [-0.0001,1]}
       foot_l_fk_ctrl.Rotation.Z {type=MovieSceneScriptingFloatChannel, default=0, const [-0.0023,1]}
       foot_l_fk_ctrl.Scale.X {type=MovieSceneScriptingFloatChannel, default=1, const [1,1]}
       foot_l_fk_ctrl.Scale.Y {type=MovieSceneScriptingFloatChannel, default=1, const [1,1]}
       foot_l_fk_ctrl.Scale.Z {type=MovieSceneScriptingFloatChannel, default=1, const [1,1]}
       ball_l_fk_ctrl.Location.X {type=MovieSceneScriptingFloatChannel, default=0, const [0,1]}
       ball_l_fk_ctrl.Location.Y {type=MovieSceneScriptingFloatChannel, default=0, const [0,1]}
       ball_l_fk_ctrl.Location.Z {type=MovieSceneScriptingFloatChannel, default=0, const [0,1]}
       ball_l_fk_ctrl.Rotation.X {type=MovieSceneScriptingFloatChannel, default=0, const [0,1]}
       ball_l_fk_ctrl.Rotation.Y {type=MovieSceneScriptingFloatChannel, default=0, const [0,1]}
       ball_l_fk_ctrl.Rotation.Z {type=MovieSceneScriptingFloatChannel, default=0, const [0,1]}
       ball_l_fk_ctrl.Scale.X {type=MovieSceneScriptingFloatChannel, default=1, const [1,1]}
       ball_l_fk_ctrl.Scale.Y {type=MovieSceneScriptingFloatChannel, default=1, const [1,1]}
       ball_l_fk_ctrl.Scale.Z {type=MovieSceneScriptingFloatChannel, default=1, const [1,1]}
       foot_l_ik_ctrl.Location.X {type=MovieSceneScriptingFloatChannel, default=-0.0026, keys [[0,-0.0026],[289,-0.0026]]}
       foot_l_ik_ctrl.Location.Y {type=MovieSceneScriptingFloatChannel, default=0.0006, keys [[0,0.0006],[289,0.0006]]}
       foot_l_ik_ctrl.Location.Z {type=MovieSceneScriptingFloatChannel, default=-0.0004, const [-0.0004,2]}
       foot_l_ik_ctrl.Rotation.X {type=MovieSceneScriptingFloatChannel, default=0.0017, keys [[0,0.0017],[289,0.0017]]}
       foot_l_ik_ctrl.Rotation.Y {type=MovieSceneScriptingFloatChannel, default=-0.0048, const [-0.0048,2]}
       foot_l_ik_ctrl.Rotation.Z {type=MovieSceneScriptingFloatChannel, default=-0.0007, keys [[0,-0.0007],[289,-0.0007]]}
       foot_l_ik_ctrl.Scale.X {type=MovieSceneScriptingFloatChannel, default=1, const [1,2]}
       foot_l_ik_ctrl.Scale.Y {type=MovieSceneScriptingFloatChannel, default=1, const [1,2]}
       foot_l_ik_ctrl.Scale.Z {type=MovieSceneScriptingFloatChannel, default=1, const [1,2]}
       ball_l_ik_ctrl.Location.X {type=MovieSceneScriptingFloatChannel, default=0}
       ball_l_ik_ctrl.Location.Y {type=MovieSceneScriptingFloatChannel, default=0}
       ball_l_ik_ctrl.Location.Z {type=MovieSceneScriptingFloatChannel, default=0}
       ball_l_ik_ctrl.Rotation.X {type=MovieSceneScriptingFloatChannel, default=0}
       ball_l_ik_ctrl.Rotation.Y {type=MovieSceneScriptingFloatChannel, default=0}
       ball_l_ik_ctrl.Rotation.Z {type=MovieSceneScriptingFloatChannel, default=0}
       ball_l_ik_ctrl.Scale.X {type=MovieSceneScriptingFloatChannel, default=1}
       ball_l_ik_ctrl.Scale.Y {type=MovieSceneScriptingFloatChannel, default=1}
       ball_l_ik_ctrl.Scale.Z {type=MovieSceneScriptingFloatChannel, default=1}
       foot_bk1_l_ctrl.Location.X {type=MovieSceneScriptingFloatChannel, default=0}
       foot_bk1_l_ctrl.Location.Y {type=MovieSceneScriptingFloatChannel, default=0}
       foot_bk1_l_ctrl.Location.Z {type=MovieSceneScriptingFloatChannel, default=0}
       foot_bk1_l_ctrl.Rotation.X {type=MovieSceneScriptingFloatChannel, default=0}
       foot_bk1_l_ctrl.Rotation.Y {type=MovieSceneScriptingFloatChannel, default=0}
       foot_bk1_l_ctrl.Rotation.Z {type=MovieSceneScriptingFloatChannel, default=0}
       foot_bk1_l_ctrl.Scale.X {type=MovieSceneScriptingFloatChannel, default=1}
       foot_bk1_l_ctrl.Scale.Y {type=MovieSceneScriptingFloatChannel, default=1}
       foot_bk1_l_ctrl.Scale.Z {type=MovieSceneScriptingFloatChannel, default=1}
       foot_roll_l_ctrl.Location.X {type=MovieSceneScriptingFloatChannel, default=0}
       foot_roll_l_ctrl.Location.Y {type=MovieSceneScriptingFloatChannel, default=0}
       foot_roll_l_ctrl.Location.Z {type=MovieSceneScriptingFloatChannel, default=0}
       foot_roll_l_ctrl.Rotation.X {type=MovieSceneScriptingFloatChannel, default=0}
       foot_roll_l_ctrl.Rotation.Y {type=MovieSceneScriptingFloatChannel, default=0}
       foot_roll_l_ctrl.Rotation.Z {type=MovieSceneScriptingFloatChannel, default=0}
       foot_roll_l_ctrl.Scale.X {type=MovieSceneScriptingFloatChannel, default=1}
       foot_roll_l_ctrl.Scale.Y {type=MovieSceneScriptingFloatChannel, default=1}
       foot_roll_l_ctrl.Scale.Z {type=MovieSceneScriptingFloatChannel, default=1}
       roll_blend_l_ctrl {type=MovieSceneScriptingFloatChannel, default=0}
       tip_l_ctrl.Location.X {type=MovieSceneScriptingFloatChannel, default=0}
       tip_l_ctrl.Location.Y {type=MovieSceneScriptingFloatChannel, default=0}
       tip_l_ctrl.Location.Z {type=MovieSceneScriptingFloatChannel, default=0}
       tip_l_ctrl.Rotation.X {type=MovieSceneScriptingFloatChannel, default=0}
       tip_l_ctrl.Rotation.Y {type=MovieSceneScriptingFloatChannel, default=0}
       tip_l_ctrl.Rotation.Z {type=MovieSceneScriptingFloatChannel, default=0}
       tip_l_ctrl.Scale.X {type=MovieSceneScriptingFloatChannel, default=1}
       tip_l_ctrl.Scale.Y {type=MovieSceneScriptingFloatChannel, default=1}
       tip_l_ctrl.Scale.Z {type=MovieSceneScriptingFloatChannel, default=1}
       heel_l_ctrl.Location.X {type=MovieSceneScriptingFloatChannel, default=0}
       heel_l_ctrl.Location.Y {type=MovieSceneScriptingFloatChannel, default=0}
       heel_l_ctrl.Location.Z {type=MovieSceneScriptingFloatChannel, default=0}
       heel_l_ctrl.Rotation.X {type=MovieSceneScriptingFloatChannel, default=0}
       heel_l_ctrl.Rotation.Y {type=MovieSceneScriptingFloatChannel, default=0}
       heel_l_ctrl.Rotation.Z {type=MovieSceneScriptingFloatChannel, default=0}
       heel_l_ctrl.Scale.X {type=MovieSceneScriptingFloatChannel, default=1}
       heel_l_ctrl.Scale.Y {type=MovieSceneScriptingFloatChannel, default=1}
       heel_l_ctrl.Scale.Z {type=MovieSceneScriptingFloatChannel, default=1}
       leg_l_softness {type=MovieSceneScriptingFloatChannel, default=0, const [0.2,3]}
       leg_l_pv_ik_ctrl.Location.X {type=MovieSceneScriptingFloatChannel, default=-0.0327}
       leg_l_pv_ik_ctrl.Location.Y {type=MovieSceneScriptingFloatChannel, default=0}
       leg_l_pv_ik_ctrl.Location.Z {type=MovieSceneScriptingFloatChannel, default=-0.0038}
       leg_l_pv_ik_ctrl.Rotation.X {type=MovieSceneScriptingFloatChannel, default=0.0051}
       leg_l_pv_ik_ctrl.Rotation.Y {type=MovieSceneScriptingFloatChannel, default=-0.0021}
       leg_l_pv_ik_ctrl.Rotation.Z {type=MovieSceneScriptingFloatChannel, default=0.0442}
       leg_l_pv_ik_ctrl.Scale.X {type=MovieSceneScriptingFloatChannel, default=1}
       leg_l_pv_ik_ctrl.Scale.Y {type=MovieSceneScriptingFloatChannel, default=1}
       leg_l_pv_ik_ctrl.Scale.Z {type=MovieSceneScriptingFloatChannel, default=1}
       thigh_r_fk_ctrl.Location.X {type=MovieSceneScriptingFloatChannel, default=0, keys [[0,0],[112,0]]}
       thigh_r_fk_ctrl.Location.Y {type=MovieSceneScriptingFloatChannel, default=0, keys [[0,0],[112,0]]}
       thigh_r_fk_ctrl.Location.Z {type=MovieSceneScriptingFloatChannel, default=0, keys [[0,0],[112,0]]}
       thigh_r_fk_ctrl.Rotation.X {type=MovieSceneScriptingFloatChannel, default=-0.0486, keys [[0,-0.0994],[112,-0.1501]]}
       thigh_r_fk_ctrl.Rotation.Y {type=MovieSceneScriptingFloatChannel, default=-0.0021, keys [[0,-0.0021],[112,-0.0021]]}
       thigh_r_fk_ctrl.Rotation.Z {type=MovieSceneScriptingFloatChannel, default=0, keys [[0,0],[112,0]]}
       thigh_r_fk_ctrl.Scale.X {type=MovieSceneScriptingFloatChannel, default=1, const [1,2]}
       thigh_r_fk_ctrl.Scale.Y {type=MovieSceneScriptingFloatChannel, default=1, const [1,2]}
       thigh_r_fk_ctrl.Scale.Z {type=MovieSceneScriptingFloatChannel, default=1, const [1,2]}
       calf_r_fk_ctrl.Location.X {type=MovieSceneScriptingFloatChannel, default=0, const [0,2]}
       calf_r_fk_ctrl.Location.Y {type=MovieSceneScriptingFloatChannel, default=0, const [0,2]}
       calf_r_fk_ctrl.Location.Z {type=MovieSceneScriptingFloatChannel, default=0, const [0,2]}
       calf_r_fk_ctrl.Rotation.X {type=MovieSceneScriptingFloatChannel, default=0, const [0,2]}
       calf_r_fk_ctrl.Rotation.Y {type=MovieSceneScriptingFloatChannel, default=0, keys [[0,0],[112,0]]}
       calf_r_fk_ctrl.Rotation.Z {type=MovieSceneScriptingFloatChannel, default=0, const [0,2]}
       calf_r_fk_ctrl.Scale.X {type=MovieSceneScriptingFloatChannel, default=1, const [1,2]}
       calf_r_fk_ctrl.Scale.Y {type=MovieSceneScriptingFloatChannel, default=1, const [1,2]}
       calf_r_fk_ctrl.Scale.Z {type=MovieSceneScriptingFloatChannel, default=1, const [1,2]}
       foot_r_fk_ctrl.Location.X {type=MovieSceneScriptingFloatChannel, default=0, keys [[0,0],[112,0]]}
       foot_r_fk_ctrl.Location.Y {type=MovieSceneScriptingFloatChannel, default=0, keys [[0,0],[112,0]]}
       foot_r_fk_ctrl.Location.Z {type=MovieSceneScriptingFloatChannel, default=0, keys [[0,0],[112,0]]}
       foot_r_fk_ctrl.Rotation.X {type=MovieSceneScriptingFloatChannel, default=0.0486, keys [[0,0.0994],[112,0.1501]]}
       foot_r_fk_ctrl.Rotation.Y {type=MovieSceneScriptingFloatChannel, default=-0.0018, keys [[0,-0.0005],[112,0.0009]]}
       foot_r_fk_ctrl.Rotation.Z {type=MovieSceneScriptingFloatChannel, default=0.0001, keys [[0,-0.0015],[112,-0.0027]]}
       foot_r_fk_ctrl.Scale.X {type=MovieSceneScriptingFloatChannel, default=1, const [1,2]}
       foot_r_fk_ctrl.Scale.Y {type=MovieSceneScriptingFloatChannel, default=1, const [1,2]}
       foot_r_fk_ctrl.Scale.Z {type=MovieSceneScriptingFloatChannel, default=1, const [1,2]}
       ball_r_fk_ctrl.Location.X {type=MovieSceneScriptingFloatChannel, default=0, const [0,2]}
       ball_r_fk_ctrl.Location.Y {type=MovieSceneScriptingFloatChannel, default=0, const [0,2]}
       ball_r_fk_ctrl.Location.Z {type=MovieSceneScriptingFloatChannel, default=0, const [0,2]}
       ball_r_fk_ctrl.Rotation.X {type=MovieSceneScriptingFloatChannel, default=0, const [0,2]}
       ball_r_fk_ctrl.Rotation.Y {type=MovieSceneScriptingFloatChannel, default=0, const [0,2]}
       ball_r_fk_ctrl.Rotation.Z {type=MovieSceneScriptingFloatChannel, default=0, const [0,2]}
       ball_r_fk_ctrl.Scale.X {type=MovieSceneScriptingFloatChannel, default=1, const [1,2]}
       ball_r_fk_ctrl.Scale.Y {type=MovieSceneScriptingFloatChannel, default=1, const [1,2]}
       ball_r_fk_ctrl.Scale.Z {type=MovieSceneScriptingFloatChannel, default=1, const [1,2]}
       foot_r_ik_ctrl.Location.X {type=MovieSceneScriptingFloatChannel, default=0, keys [[0,-0.0027],[289,-0.0027]]}
       foot_r_ik_ctrl.Location.Y {type=MovieSceneScriptingFloatChannel, default=0, keys [[0,-0.001],[289,-0.001]]}
       foot_r_ik_ctrl.Location.Z {type=MovieSceneScriptingFloatChannel, default=0, keys [[0,-0.0002],[289,-0.0002]]}
       foot_r_ik_ctrl.Rotation.X {type=MovieSceneScriptingFloatChannel, default=0, const [0.0022,2]}
       foot_r_ik_ctrl.Rotation.Y {type=MovieSceneScriptingFloatChannel, default=0, keys [[0,-0.0049],[289,-0.0049]]}
       foot_r_ik_ctrl.Rotation.Z {type=MovieSceneScriptingFloatChannel, default=0, keys [[0,-0.0007],[289,-0.0007]]}
       foot_r_ik_ctrl.Scale.X {type=MovieSceneScriptingFloatChannel, default=1, const [1,2]}
       foot_r_ik_ctrl.Scale.Y {type=MovieSceneScriptingFloatChannel, default=1, const [1,2]}
       foot_r_ik_ctrl.Scale.Z {type=MovieSceneScriptingFloatChannel, default=1, const [1,2]}
       ball_r_ik_ctrl.Location.X {type=MovieSceneScriptingFloatChannel, default=0, const [0,2]}
       ball_r_ik_ctrl.Location.Y {type=MovieSceneScriptingFloatChannel, default=0, const [0,2]}
       ball_r_ik_ctrl.Location.Z {type=MovieSceneScriptingFloatChannel, default=0, const [0,2]}
       ball_r_ik_ctrl.Rotation.X {type=MovieSceneScriptingFloatChannel, default=0, const [0,2]}
       ball_r_ik_ctrl.Rotation.Y {type=MovieSceneScriptingFloatChannel, default=0, const [0,2]}
       ball_r_ik_ctrl.Rotation.Z {type=MovieSceneScriptingFloatChannel, default=0, const [0,2]}
       ball_r_ik_ctrl.Scale.X {type=MovieSceneScriptingFloatChannel, default=1, const [1,2]}
       ball_r_ik_ctrl.Scale.Y {type=MovieSceneScriptingFloatChannel, default=1, const [1,2]}
       ball_r_ik_ctrl.Scale.Z {type=MovieSceneScriptingFloatChannel, default=1, const [1,2]}
       foot_bk1_r_ctrl.Location.X {type=MovieSceneScriptingFloatChannel, default=0, const [0,2]}
       foot_bk1_r_ctrl.Location.Y {type=MovieSceneScriptingFloatChannel, default=0, const [0,2]}
       foot_bk1_r_ctrl.Location.Z {type=MovieSceneScriptingFloatChannel, default=0, const [0,2]}
       foot_bk1_r_ctrl.Rotation.X {type=MovieSceneScriptingFloatChannel, default=0, const [0,2]}
       foot_bk1_r_ctrl.Rotation.Y {type=MovieSceneScriptingFloatChannel, default=0, const [0,2]}
       foot_bk1_r_ctrl.Rotation.Z {type=MovieSceneScriptingFloatChannel, default=0, const [0,2]}
       foot_bk1_r_ctrl.Scale.X {type=MovieSceneScriptingFloatChannel, default=1, const [1,2]}
       foot_bk1_r_ctrl.Scale.Y {type=MovieSceneScriptingFloatChannel, default=1, const [1,2]}
       foot_bk1_r_ctrl.Scale.Z {type=MovieSceneScriptingFloatChannel, default=1, const [1,2]}
       foot_roll_r_ctrl.Location.X {type=MovieSceneScriptingFloatChannel, default=0, const [0,2]}
       foot_roll_r_ctrl.Location.Y {type=MovieSceneScriptingFloatChannel, default=0, const [0,2]}
       foot_roll_r_ctrl.Location.Z {type=MovieSceneScriptingFloatChannel, default=0, const [0,2]}
       foot_roll_r_ctrl.Rotation.X {type=MovieSceneScriptingFloatChannel, default=0, const [0,2]}
       foot_roll_r_ctrl.Rotation.Y {type=MovieSceneScriptingFloatChannel, default=0, const [0,2]}
       foot_roll_r_ctrl.Rotation.Z {type=MovieSceneScriptingFloatChannel, default=0, const [0,2]}
       foot_roll_r_ctrl.Scale.X {type=MovieSceneScriptingFloatChannel, default=1, const [1,2]}
       foot_roll_r_ctrl.Scale.Y {type=MovieSceneScriptingFloatChannel, default=1, const [1,2]}
       foot_roll_r_ctrl.Scale.Z {type=MovieSceneScriptingFloatChannel, default=1, const [1,2]}
       roll_blend_r_ctrl {type=MovieSceneScriptingFloatChannel, default=0}
       tip_r_ctrl.Location.X {type=MovieSceneScriptingFloatChannel, default=0, const [0,2]}
       tip_r_ctrl.Location.Y {type=MovieSceneScriptingFloatChannel, default=0, const [0,2]}
       tip_r_ctrl.Location.Z {type=MovieSceneScriptingFloatChannel, default=0, const [0,2]}
       tip_r_ctrl.Rotation.X {type=MovieSceneScriptingFloatChannel, default=0, const [0,2]}
       tip_r_ctrl.Rotation.Y {type=MovieSceneScriptingFloatChannel, default=0, const [0,2]}
       tip_r_ctrl.Rotation.Z {type=MovieSceneScriptingFloatChannel, default=0, const [0,2]}
       tip_r_ctrl.Scale.X {type=MovieSceneScriptingFloatChannel, default=1, const [1,2]}
       tip_r_ctrl.Scale.Y {type=MovieSceneScriptingFloatChannel, default=1, const [1,2]}
       tip_r_ctrl.Scale.Z {type=MovieSceneScriptingFloatChannel, default=1, const [1,2]}
       heel_r_ctrl.Location.X {type=MovieSceneScriptingFloatChannel, default=0, const [0,2]}
       heel_r_ctrl.Location.Y {type=MovieSceneScriptingFloatChannel, default=0, const [0,2]}
       heel_r_ctrl.Location.Z {type=MovieSceneScriptingFloatChannel, default=0, const [0,2]}
       heel_r_ctrl.Rotation.X {type=MovieSceneScriptingFloatChannel, default=0, const [0,2]}
       heel_r_ctrl.Rotation.Y {type=MovieSceneScriptingFloatChannel, default=0, const [0,2]}
       heel_r_ctrl.Rotation.Z {type=MovieSceneScriptingFloatChannel, default=0, const [0,2]}
       heel_r_ctrl.Scale.X {type=MovieSceneScriptingFloatChannel, default=1, const [1,2]}
       heel_r_ctrl.Scale.Y {type=MovieSceneScriptingFloatChannel, default=1, const [1,2]}
       heel_r_ctrl.Scale.Z {type=MovieSceneScriptingFloatChannel, default=1, const [1,2]}
       leg_r_softness {type=MovieSceneScriptingFloatChannel, default=0, const [0.15,3]}
       leg_r_pv_ik_ctrl.Location.X {type=MovieSceneScriptingFloatChannel, default=-0.0358, keys [[0,-0.0358],[112,0]]}
       leg_r_pv_ik_ctrl.Location.Y {type=MovieSceneScriptingFloatChannel, default=0, keys [[0,0],[112,0]]}
       leg_r_pv_ik_ctrl.Location.Z {type=MovieSceneScriptingFloatChannel, default=-0.0041, keys [[0,-0.0041],[112,0]]}
       leg_r_pv_ik_ctrl.Rotation.X {type=MovieSceneScriptingFloatChannel, default=0.0055, keys [[0,0.0055],[112,0]]}
       leg_r_pv_ik_ctrl.Rotation.Y {type=MovieSceneScriptingFloatChannel, default=-0.0023, keys [[0,-0.0023],[112,0]]}
       leg_r_pv_ik_ctrl.Rotation.Z {type=MovieSceneScriptingFloatChannel, default=0.0483, keys [[0,0.0483],[112,0]]}
       leg_r_pv_ik_ctrl.Scale.X {type=MovieSceneScriptingFloatChannel, default=1, const [1,2]}
       leg_r_pv_ik_ctrl.Scale.Y {type=MovieSceneScriptingFloatChannel, default=1, const [1,2]}
       leg_r_pv_ik_ctrl.Scale.Z {type=MovieSceneScriptingFloatChannel, default=1, const [1,2]}
       upperarm_l_fk_ctrl.Location.X {type=MovieSceneScriptingFloatChannel, default=0}
       upperarm_l_fk_ctrl.Location.Y {type=MovieSceneScriptingFloatChannel, default=0}
       upperarm_l_fk_ctrl.Location.Z {type=MovieSceneScriptingFloatChannel, default=0}
       upperarm_l_fk_ctrl.Rotation.X {type=MovieSceneScriptingFloatChannel, default=0}
       upperarm_l_fk_ctrl.Rotation.Y {type=MovieSceneScriptingFloatChannel, default=0}
       upperarm_l_fk_ctrl.Rotation.Z {type=MovieSceneScriptingFloatChannel, default=0}
       upperarm_l_fk_ctrl.Scale.X {type=MovieSceneScriptingFloatChannel, default=1}
       upperarm_l_fk_ctrl.Scale.Y {type=MovieSceneScriptingFloatChannel, default=1}
       upperarm_l_fk_ctrl.Scale.Z {type=MovieSceneScriptingFloatChannel, default=1}
       lowerarm_l_fk_ctrl.Location.X {type=MovieSceneScriptingFloatChannel, default=0, const [0,5]}
       lowerarm_l_fk_ctrl.Location.Y {type=MovieSceneScriptingFloatChannel, default=0, const [0,5]}
       lowerarm_l_fk_ctrl.Location.Z {type=MovieSceneScriptingFloatChannel, default=0, const [0,5]}
       lowerarm_l_fk_ctrl.Rotation.X {type=MovieSceneScriptingFloatChannel, default=0, keys [[114,0],[116,0.0218],[118,0.0872],[119,0],[148,0]]}
       lowerarm_l_fk_ctrl.Rotation.Y {type=MovieSceneScriptingFloatChannel, default=0, keys [[114,0],[116,0.6905],[118,2.762],[119,0],[148,0]]}
       lowerarm_l_fk_ctrl.Rotation.Z {type=MovieSceneScriptingFloatChannel, default=0, keys [[114,0],[116,0.4522],[118,1.8086],[119,0],[148,0]]}
       lowerarm_l_fk_ctrl.Scale.X {type=MovieSceneScriptingFloatChannel, default=1, const [1,5]}
       lowerarm_l_fk_ctrl.Scale.Y {type=MovieSceneScriptingFloatChannel, default=1, const [1,5]}
       lowerarm_l_fk_ctrl.Scale.Z {type=MovieSceneScriptingFloatChannel, default=1, const [1,5]}
       hand_l_fk_ctrl.Location.X {type=MovieSceneScriptingFloatChannel, default=0, const [0,3]}
       hand_l_fk_ctrl.Location.Y {type=MovieSceneScriptingFloatChannel, default=0, const [0,3]}
       hand_l_fk_ctrl.Location.Z {type=MovieSceneScriptingFloatChannel, default=0, const [0,3]}
       hand_l_fk_ctrl.Rotation.X {type=MovieSceneScriptingFloatChannel, default=0, keys [[200,0],[202,8.9449],[204,1.1181],[208,0]]}
       hand_l_fk_ctrl.Rotation.Y {type=MovieSceneScriptingFloatChannel, default=0, keys [[200,0],[204,5.0787],[208,0]]}
       hand_l_fk_ctrl.Rotation.Z {type=MovieSceneScriptingFloatChannel, default=0, keys [[200,0],[202,-8.0645],[204,-1.0081],[208,0]]}
       hand_l_fk_ctrl.Scale.X {type=MovieSceneScriptingFloatChannel, default=1, const [1,3]}
       hand_l_fk_ctrl.Scale.Y {type=MovieSceneScriptingFloatChannel, default=1, const [1,3]}
       hand_l_fk_ctrl.Scale.Z {type=MovieSceneScriptingFloatChannel, default=1, const [1,3]}
       thumb_01_l_ctrl.Location.X {type=MovieSceneScriptingFloatChannel, default=0}
       thumb_01_l_ctrl.Location.Y {type=MovieSceneScriptingFloatChannel, default=0}
       thumb_01_l_ctrl.Location.Z {type=MovieSceneScriptingFloatChannel, default=0}
       thumb_01_l_ctrl.Rotation.X {type=MovieSceneScriptingFloatChannel, default=0}
       thumb_01_l_ctrl.Rotation.Y {type=MovieSceneScriptingFloatChannel, default=0}
       thumb_01_l_ctrl.Rotation.Z {type=MovieSceneScriptingFloatChannel, default=0}
       thumb_01_l_ctrl.Scale.X {type=MovieSceneScriptingFloatChannel, default=1}
       thumb_01_l_ctrl.Scale.Y {type=MovieSceneScriptingFloatChannel, default=1}
       thumb_01_l_ctrl.Scale.Z {type=MovieSceneScriptingFloatChannel, default=1}
       thumb_02_l_ctrl.Location.X {type=MovieSceneScriptingFloatChannel, default=0}
       thumb_02_l_ctrl.Location.Y {type=MovieSceneScriptingFloatChannel, default=0}
       thumb_02_l_ctrl.Location.Z {type=MovieSceneScriptingFloatChannel, default=0}
       thumb_02_l_ctrl.Rotation.X {type=MovieSceneScriptingFloatChannel, default=0}
       thumb_02_l_ctrl.Rotation.Y {type=MovieSceneScriptingFloatChannel, default=0}
       thumb_02_l_ctrl.Rotation.Z {type=MovieSceneScriptingFloatChannel, default=0}
       thumb_02_l_ctrl.Scale.X {type=MovieSceneScriptingFloatChannel, default=1}
       thumb_02_l_ctrl.Scale.Y {type=MovieSceneScriptingFloatChannel, default=1}
       thumb_02_l_ctrl.Scale.Z {type=MovieSceneScriptingFloatChannel, default=1}
       thumb_03_l_ctrl.Location.X {type=MovieSceneScriptingFloatChannel, default=0}
       thumb_03_l_ctrl.Location.Y {type=MovieSceneScriptingFloatChannel, default=0}
       thumb_03_l_ctrl.Location.Z {type=MovieSceneScriptingFloatChannel, default=0}
       thumb_03_l_ctrl.Rotation.X {type=MovieSceneScriptingFloatChannel, default=0}
       thumb_03_l_ctrl.Rotation.Y {type=MovieSceneScriptingFloatChannel, default=0}
       thumb_03_l_ctrl.Rotation.Z {type=MovieSceneScriptingFloatChannel, default=0}
       thumb_03_l_ctrl.Scale.X {type=MovieSceneScriptingFloatChannel, default=1}
       thumb_03_l_ctrl.Scale.Y {type=MovieSceneScriptingFloatChannel, default=1}
       thumb_03_l_ctrl.Scale.Z {type=MovieSceneScriptingFloatChannel, default=1}
       index_metacarpal_l_ctrl.Location.X {type=MovieSceneScriptingFloatChannel, default=0}
       index_metacarpal_l_ctrl.Location.Y {type=MovieSceneScriptingFloatChannel, default=0}
       index_metacarpal_l_ctrl.Location.Z {type=MovieSceneScriptingFloatChannel, default=0}
       index_metacarpal_l_ctrl.Rotation.X {type=MovieSceneScriptingFloatChannel, default=0}
       index_metacarpal_l_ctrl.Rotation.Y {type=MovieSceneScriptingFloatChannel, default=0}
       index_metacarpal_l_ctrl.Rotation.Z {type=MovieSceneScriptingFloatChannel, default=0}
       index_metacarpal_l_ctrl.Scale.X {type=MovieSceneScriptingFloatChannel, default=1}
       index_metacarpal_l_ctrl.Scale.Y {type=MovieSceneScriptingFloatChannel, default=1}
       index_metacarpal_l_ctrl.Scale.Z {type=MovieSceneScriptingFloatChannel, default=1}
       index_01_l_ctrl.Location.X {type=MovieSceneScriptingFloatChannel, default=0}
       index_01_l_ctrl.Location.Y {type=MovieSceneScriptingFloatChannel, default=0}
       index_01_l_ctrl.Location.Z {type=MovieSceneScriptingFloatChannel, default=0}
       index_01_l_ctrl.Rotation.X {type=MovieSceneScriptingFloatChannel, default=0}
       index_01_l_ctrl.Rotation.Y {type=MovieSceneScriptingFloatChannel, default=0}
       index_01_l_ctrl.Rotation.Z {type=MovieSceneScriptingFloatChannel, default=0}
       index_01_l_ctrl.Scale.X {type=MovieSceneScriptingFloatChannel, default=1}
       index_01_l_ctrl.Scale.Y {type=MovieSceneScriptingFloatChannel, default=1}
       index_01_l_ctrl.Scale.Z {type=MovieSceneScriptingFloatChannel, default=1}
       index_02_l_ctrl.Location.X {type=MovieSceneScriptingFloatChannel, default=0}
       index_02_l_ctrl.Location.Y {type=MovieSceneScriptingFloatChannel, default=0}
       index_02_l_ctrl.Location.Z {type=MovieSceneScriptingFloatChannel, default=0}
       index_02_l_ctrl.Rotation.X {type=MovieSceneScriptingFloatChannel, default=0}
       index_02_l_ctrl.Rotation.Y {type=MovieSceneScriptingFloatChannel, default=0}
       index_02_l_ctrl.Rotation.Z {type=MovieSceneScriptingFloatChannel, default=0}
       index_02_l_ctrl.Scale.X {type=MovieSceneScriptingFloatChannel, default=1}
       index_02_l_ctrl.Scale.Y {type=MovieSceneScriptingFloatChannel, default=1}
       index_02_l_ctrl.Scale.Z {type=MovieSceneScriptingFloatChannel, default=1}
       index_03_l_ctrl.Location.X {type=MovieSceneScriptingFloatChannel, default=0}
       index_03_l_ctrl.Location.Y {type=MovieSceneScriptingFloatChannel, default=0}
       index_03_l_ctrl.Location.Z {type=MovieSceneScriptingFloatChannel, default=0}
       index_03_l_ctrl.Rotation.X {type=MovieSceneScriptingFloatChannel, default=0}
       index_03_l_ctrl.Rotation.Y {type=MovieSceneScriptingFloatChannel, default=0}
       index_03_l_ctrl.Rotation.Z {type=MovieSceneScriptingFloatChannel, default=0}
       index_03_l_ctrl.Scale.X {type=MovieSceneScriptingFloatChannel, default=1}
       index_03_l_ctrl.Scale.Y {type=MovieSceneScriptingFloatChannel, default=1}
       index_03_l_ctrl.Scale.Z {type=MovieSceneScriptingFloatChannel, default=1}
       middle_metacarpal_l_ctrl.Location.X {type=MovieSceneScriptingFloatChannel, default=0}
       middle_metacarpal_l_ctrl.Location.Y {type=MovieSceneScriptingFloatChannel, default=0}
       middle_metacarpal_l_ctrl.Location.Z {type=MovieSceneScriptingFloatChannel, default=0}
       middle_metacarpal_l_ctrl.Rotation.X {type=MovieSceneScriptingFloatChannel, default=0}
       middle_metacarpal_l_ctrl.Rotation.Y {type=MovieSceneScriptingFloatChannel, default=0}
       middle_metacarpal_l_ctrl.Rotation.Z {type=MovieSceneScriptingFloatChannel, default=0}
       middle_metacarpal_l_ctrl.Scale.X {type=MovieSceneScriptingFloatChannel, default=1}
       middle_metacarpal_l_ctrl.Scale.Y {type=MovieSceneScriptingFloatChannel, default=1}
       middle_metacarpal_l_ctrl.Scale.Z {type=MovieSceneScriptingFloatChannel, default=1}
       middle_01_l_ctrl.Location.X {type=MovieSceneScriptingFloatChannel, default=0}
       middle_01_l_ctrl.Location.Y {type=MovieSceneScriptingFloatChannel, default=0}
       middle_01_l_ctrl.Location.Z {type=MovieSceneScriptingFloatChannel, default=0}
       middle_01_l_ctrl.Rotation.X {type=MovieSceneScriptingFloatChannel, default=0}
       middle_01_l_ctrl.Rotation.Y {type=MovieSceneScriptingFloatChannel, default=0}
       middle_01_l_ctrl.Rotation.Z {type=MovieSceneScriptingFloatChannel, default=0}
       middle_01_l_ctrl.Scale.X {type=MovieSceneScriptingFloatChannel, default=1}
       middle_01_l_ctrl.Scale.Y {type=MovieSceneScriptingFloatChannel, default=1}
       middle_01_l_ctrl.Scale.Z {type=MovieSceneScriptingFloatChannel, default=1}
       middle_02_l_ctrl.Location.X {type=MovieSceneScriptingFloatChannel, default=0}
       middle_02_l_ctrl.Location.Y {type=MovieSceneScriptingFloatChannel, default=0}
       middle_02_l_ctrl.Location.Z {type=MovieSceneScriptingFloatChannel, default=0}
       middle_02_l_ctrl.Rotation.X {type=MovieSceneScriptingFloatChannel, default=0}
       middle_02_l_ctrl.Rotation.Y {type=MovieSceneScriptingFloatChannel, default=0}
       middle_02_l_ctrl.Rotation.Z {type=MovieSceneScriptingFloatChannel, default=0}
       middle_02_l_ctrl.Scale.X {type=MovieSceneScriptingFloatChannel, default=1}
       middle_02_l_ctrl.Scale.Y {type=MovieSceneScriptingFloatChannel, default=1}
       middle_02_l_ctrl.Scale.Z {type=MovieSceneScriptingFloatChannel, default=1}
       middle_03_l_ctrl.Location.X {type=MovieSceneScriptingFloatChannel, default=0}
       middle_03_l_ctrl.Location.Y {type=MovieSceneScriptingFloatChannel, default=0}
       middle_03_l_ctrl.Location.Z {type=MovieSceneScriptingFloatChannel, default=0}
       middle_03_l_ctrl.Rotation.X {type=MovieSceneScriptingFloatChannel, default=0}
       middle_03_l_ctrl.Rotation.Y {type=MovieSceneScriptingFloatChannel, default=0}
       middle_03_l_ctrl.Rotation.Z {type=MovieSceneScriptingFloatChannel, default=0}
       middle_03_l_ctrl.Scale.X {type=MovieSceneScriptingFloatChannel, default=1}
       middle_03_l_ctrl.Scale.Y {type=MovieSceneScriptingFloatChannel, default=1}
       middle_03_l_ctrl.Scale.Z {type=MovieSceneScriptingFloatChannel, default=1}
       ring_metacarpal_l_ctrl.Location.X {type=MovieSceneScriptingFloatChannel, default=0}
       ring_metacarpal_l_ctrl.Location.Y {type=MovieSceneScriptingFloatChannel, default=0}
       ring_metacarpal_l_ctrl.Location.Z {type=MovieSceneScriptingFloatChannel, default=0}
       ring_metacarpal_l_ctrl.Rotation.X {type=MovieSceneScriptingFloatChannel, default=0}
       ring_metacarpal_l_ctrl.Rotation.Y {type=MovieSceneScriptingFloatChannel, default=0}
       ring_metacarpal_l_ctrl.Rotation.Z {type=MovieSceneScriptingFloatChannel, default=0}
       ring_metacarpal_l_ctrl.Scale.X {type=MovieSceneScriptingFloatChannel, default=1}
       ring_metacarpal_l_ctrl.Scale.Y {type=MovieSceneScriptingFloatChannel, default=1}
       ring_metacarpal_l_ctrl.Scale.Z {type=MovieSceneScriptingFloatChannel, default=1}
       ring_01_l_ctrl.Location.X {type=MovieSceneScriptingFloatChannel, default=0}
       ring_01_l_ctrl.Location.Y {type=MovieSceneScriptingFloatChannel, default=0}
       ring_01_l_ctrl.Location.Z {type=MovieSceneScriptingFloatChannel, default=0}
       ring_01_l_ctrl.Rotation.X {type=MovieSceneScriptingFloatChannel, default=0}
       ring_01_l_ctrl.Rotation.Y {type=MovieSceneScriptingFloatChannel, default=0}
       ring_01_l_ctrl.Rotation.Z {type=MovieSceneScriptingFloatChannel, default=0}
       ring_01_l_ctrl.Scale.X {type=MovieSceneScriptingFloatChannel, default=1}
       ring_01_l_ctrl.Scale.Y {type=MovieSceneScriptingFloatChannel, default=1}
       ring_01_l_ctrl.Scale.Z {type=MovieSceneScriptingFloatChannel, default=1}
       ring_02_l_ctrl.Location.X {type=MovieSceneScriptingFloatChannel, default=0}
       ring_02_l_ctrl.Location.Y {type=MovieSceneScriptingFloatChannel, default=0}
       ring_02_l_ctrl.Location.Z {type=MovieSceneScriptingFloatChannel, default=0}
       ring_02_l_ctrl.Rotation.X {type=MovieSceneScriptingFloatChannel, default=0}
       ring_02_l_ctrl.Rotation.Y {type=MovieSceneScriptingFloatChannel, default=0}
       ring_02_l_ctrl.Rotation.Z {type=MovieSceneScriptingFloatChannel, default=0}
       ring_02_l_ctrl.Scale.X {type=MovieSceneScriptingFloatChannel, default=1}
       ring_02_l_ctrl.Scale.Y {type=MovieSceneScriptingFloatChannel, default=1}
       ring_02_l_ctrl.Scale.Z {type=MovieSceneScriptingFloatChannel, default=1}
       ring_03_l_ctrl.Location.X {type=MovieSceneScriptingFloatChannel, default=0}
       ring_03_l_ctrl.Location.Y {type=MovieSceneScriptingFloatChannel, default=0}
       ring_03_l_ctrl.Location.Z {type=MovieSceneScriptingFloatChannel, default=0}
       ring_03_l_ctrl.Rotation.X {type=MovieSceneScriptingFloatChannel, default=0}
       ring_03_l_ctrl.Rotation.Y {type=MovieSceneScriptingFloatChannel, default=0}
       ring_03_l_ctrl.Rotation.Z {type=MovieSceneScriptingFloatChannel, default=0}
       ring_03_l_ctrl.Scale.X {type=MovieSceneScriptingFloatChannel, default=1}
       ring_03_l_ctrl.Scale.Y {type=MovieSceneScriptingFloatChannel, default=1}
       ring_03_l_ctrl.Scale.Z {type=MovieSceneScriptingFloatChannel, default=1}
       pinky_metacarpal_l_ctrl.Location.X {type=MovieSceneScriptingFloatChannel, default=0}
       pinky_metacarpal_l_ctrl.Location.Y {type=MovieSceneScriptingFloatChannel, default=0}
       pinky_metacarpal_l_ctrl.Location.Z {type=MovieSceneScriptingFloatChannel, default=0}
       pinky_metacarpal_l_ctrl.Rotation.X {type=MovieSceneScriptingFloatChannel, default=0}
       pinky_metacarpal_l_ctrl.Rotation.Y {type=MovieSceneScriptingFloatChannel, default=0}
       pinky_metacarpal_l_ctrl.Rotation.Z {type=MovieSceneScriptingFloatChannel, default=0}
       pinky_metacarpal_l_ctrl.Scale.X {type=MovieSceneScriptingFloatChannel, default=1}
       pinky_metacarpal_l_ctrl.Scale.Y {type=MovieSceneScriptingFloatChannel, default=1}
       pinky_metacarpal_l_ctrl.Scale.Z {type=MovieSceneScriptingFloatChannel, default=1}
       pinky_01_l_ctrl.Location.X {type=MovieSceneScriptingFloatChannel, default=0}
       pinky_01_l_ctrl.Location.Y {type=MovieSceneScriptingFloatChannel, default=0}
       pinky_01_l_ctrl.Location.Z {type=MovieSceneScriptingFloatChannel, default=0}
       pinky_01_l_ctrl.Rotation.X {type=MovieSceneScriptingFloatChannel, default=0}
       pinky_01_l_ctrl.Rotation.Y {type=MovieSceneScriptingFloatChannel, default=0}
       pinky_01_l_ctrl.Rotation.Z {type=MovieSceneScriptingFloatChannel, default=0}
       pinky_01_l_ctrl.Scale.X {type=MovieSceneScriptingFloatChannel, default=1}
       pinky_01_l_ctrl.Scale.Y {type=MovieSceneScriptingFloatChannel, default=1}
       pinky_01_l_ctrl.Scale.Z {type=MovieSceneScriptingFloatChannel, default=1}
       pinky_02_l_ctrl.Location.X {type=MovieSceneScriptingFloatChannel, default=0}
       pinky_02_l_ctrl.Location.Y {type=MovieSceneScriptingFloatChannel, default=0}
       pinky_02_l_ctrl.Location.Z {type=MovieSceneScriptingFloatChannel, default=0}
       pinky_02_l_ctrl.Rotation.X {type=MovieSceneScriptingFloatChannel, default=0}
       pinky_02_l_ctrl.Rotation.Y {type=MovieSceneScriptingFloatChannel, default=0}
       pinky_02_l_ctrl.Rotation.Z {type=MovieSceneScriptingFloatChannel, default=0}
       pinky_02_l_ctrl.Scale.X {type=MovieSceneScriptingFloatChannel, default=1}
       pinky_02_l_ctrl.Scale.Y {type=MovieSceneScriptingFloatChannel, default=1}
       pinky_02_l_ctrl.Scale.Z {type=MovieSceneScriptingFloatChannel, default=1}
       pinky_03_l_ctrl.Location.X {type=MovieSceneScriptingFloatChannel, default=0}
       pinky_03_l_ctrl.Location.Y {type=MovieSceneScriptingFloatChannel, default=0}
       pinky_03_l_ctrl.Location.Z {type=MovieSceneScriptingFloatChannel, default=0}
       pinky_03_l_ctrl.Rotation.X {type=MovieSceneScriptingFloatChannel, default=0}
       pinky_03_l_ctrl.Rotation.Y {type=MovieSceneScriptingFloatChannel, default=0}
       pinky_03_l_ctrl.Rotation.Z {type=MovieSceneScriptingFloatChannel, default=0}
       pinky_03_l_ctrl.Scale.X {type=MovieSceneScriptingFloatChannel, default=1}
       pinky_03_l_ctrl.Scale.Y {type=MovieSceneScriptingFloatChannel, default=1}
       pinky_03_l_ctrl.Scale.Z {type=MovieSceneScriptingFloatChannel, default=1}
       hand_l_ik_ctrl.Location.X {type=MovieSceneScriptingFloatChannel, default=0}
       hand_l_ik_ctrl.Location.Y {type=MovieSceneScriptingFloatChannel, default=0}
       hand_l_ik_ctrl.Location.Z {type=MovieSceneScriptingFloatChannel, default=0}
       hand_l_ik_ctrl.Rotation.X {type=MovieSceneScriptingFloatChannel, default=0}
       hand_l_ik_ctrl.Rotation.Y {type=MovieSceneScriptingFloatChannel, default=0}
       hand_l_ik_ctrl.Rotation.Z {type=MovieSceneScriptingFloatChannel, default=0}
       hand_l_ik_ctrl.Scale.X {type=MovieSceneScriptingFloatChannel, default=1}
       hand_l_ik_ctrl.Scale.Y {type=MovieSceneScriptingFloatChannel, default=1}
       hand_l_ik_ctrl.Scale.Z {type=MovieSceneScriptingFloatChannel, default=1}
       Softness {type=MovieSceneScriptingFloatChannel, default=0}
       hand_l_ikorient_ctrl.Rotation.Roll {type=MovieSceneScriptingFloatChannel, default=0}
       hand_l_ikorient_ctrl.Rotation.Pitch {type=MovieSceneScriptingFloatChannel, default=0}
       hand_l_ikorient_ctrl.Rotation.Yaw {type=MovieSceneScriptingFloatChannel, default=0}
       arm_l_pv_ik_ctrl.Location.X {type=MovieSceneScriptingFloatChannel, default=0}
       arm_l_pv_ik_ctrl.Location.Y {type=MovieSceneScriptingFloatChannel, default=0}
       arm_l_pv_ik_ctrl.Location.Z {type=MovieSceneScriptingFloatChannel, default=0}
       arm_l_pv_ik_ctrl.Rotation.X {type=MovieSceneScriptingFloatChannel, default=0}
       arm_l_pv_ik_ctrl.Rotation.Y {type=MovieSceneScriptingFloatChannel, default=0}
       arm_l_pv_ik_ctrl.Rotation.Z {type=MovieSceneScriptingFloatChannel, default=0}
       arm_l_pv_ik_ctrl.Scale.X {type=MovieSceneScriptingFloatChannel, default=1}
       arm_l_pv_ik_ctrl.Scale.Y {type=MovieSceneScriptingFloatChannel, default=1}
       arm_l_pv_ik_ctrl.Scale.Z {type=MovieSceneScriptingFloatChannel, default=1}
       upperarm_r_fk_ctrl.Location.X {type=MovieSceneScriptingFloatChannel, default=0, const [0,1]}
       upperarm_r_fk_ctrl.Location.Y {type=MovieSceneScriptingFloatChannel, default=0, const [0,1]}
       upperarm_r_fk_ctrl.Location.Z {type=MovieSceneScriptingFloatChannel, default=0, const [0,1]}
       upperarm_r_fk_ctrl.Rotation.X {type=MovieSceneScriptingFloatChannel, default=3.1079, const [3.1079,1]}
       upperarm_r_fk_ctrl.Rotation.Y {type=MovieSceneScriptingFloatChannel, default=-1.5045, const [-1.5045,1]}
       upperarm_r_fk_ctrl.Rotation.Z {type=MovieSceneScriptingFloatChannel, default=-5.3847, const [-5.3847,1]}
       upperarm_r_fk_ctrl.Scale.X {type=MovieSceneScriptingFloatChannel, default=1, const [1,1]}
       upperarm_r_fk_ctrl.Scale.Y {type=MovieSceneScriptingFloatChannel, default=1, const [1,1]}
       upperarm_r_fk_ctrl.Scale.Z {type=MovieSceneScriptingFloatChannel, default=1, const [1,1]}
       lowerarm_r_fk_ctrl.Location.X {type=MovieSceneScriptingFloatChannel, default=0, thinned_from=3, keys [[114,0],[145,0]]}
       lowerarm_r_fk_ctrl.Location.Y {type=MovieSceneScriptingFloatChannel, default=-0.0001, thinned_from=3, keys [[114,-0.0001],[145,-0.0001]]}
       lowerarm_r_fk_ctrl.Location.Z {type=MovieSceneScriptingFloatChannel, default=0, thinned_from=3, keys [[114,0],[145,0]]}
       lowerarm_r_fk_ctrl.Rotation.X {type=MovieSceneScriptingFloatChannel, default=6.226, const [6.226,3]}
       lowerarm_r_fk_ctrl.Rotation.Y {type=MovieSceneScriptingFloatChannel, default=2.9856, thinned_from=3, keys [[114,2.9856],[145,2.9856]]}
       lowerarm_r_fk_ctrl.Rotation.Z {type=MovieSceneScriptingFloatChannel, default=2.7402, thinned_from=3, keys [[114,2.7402],[145,2.7402]]}
       lowerarm_r_fk_ctrl.Scale.X {type=MovieSceneScriptingFloatChannel, default=1, const [1,3]}
       lowerarm_r_fk_ctrl.Scale.Y {type=MovieSceneScriptingFloatChannel, default=1, const [1,3]}
       lowerarm_r_fk_ctrl.Scale.Z {type=MovieSceneScriptingFloatChannel, default=1, const [1,3]}
       hand_r_fk_ctrl.Location.X {type=MovieSceneScriptingFloatChannel, default=1.6754, thinned_from=5, keys [[115,1.6754],[224,1.6754]]}
       hand_r_fk_ctrl.Location.Y {type=MovieSceneScriptingFloatChannel, default=-1.0986, const [-1.0986,5]}
       hand_r_fk_ctrl.Location.Z {type=MovieSceneScriptingFloatChannel, default=1.0646, thinned_from=5, keys [[115,1.0646],[224,1.0646]]}
       hand_r_fk_ctrl.Rotation.X {type=MovieSceneScriptingFloatChannel, default=-1.4262, thinned_from=5, keys [[115,-0.5607],[122,-1.4262],[217,-1.4262],[224,-0.5607]]}
       hand_r_fk_ctrl.Rotation.Y {type=MovieSceneScriptingFloatChannel, default=1.6479, thinned_from=5, keys [[115,-6.3545],[122,1.6479],[217,1.6479],[224,-6.3545]]}
       hand_r_fk_ctrl.Rotation.Z {type=MovieSceneScriptingFloatChannel, default=3.8968, thinned_from=5, keys [[115,4.004],[122,3.8968],[217,3.8968],[224,4.004]]}
       hand_r_fk_ctrl.Scale.X {type=MovieSceneScriptingFloatChannel, default=1, const [1,5]}
       hand_r_fk_ctrl.Scale.Y {type=MovieSceneScriptingFloatChannel, default=1, const [1,5]}
       hand_r_fk_ctrl.Scale.Z {type=MovieSceneScriptingFloatChannel, default=1, const [1,5]}
       thumb_01_r_ctrl.Location.X {type=MovieSceneScriptingFloatChannel, default=0}
       thumb_01_r_ctrl.Location.Y {type=MovieSceneScriptingFloatChannel, default=0}
       thumb_01_r_ctrl.Location.Z {type=MovieSceneScriptingFloatChannel, default=0}
       thumb_01_r_ctrl.Rotation.X {type=MovieSceneScriptingFloatChannel, default=0}
       thumb_01_r_ctrl.Rotation.Y {type=MovieSceneScriptingFloatChannel, default=0}
       thumb_01_r_ctrl.Rotation.Z {type=MovieSceneScriptingFloatChannel, default=0}
       thumb_01_r_ctrl.Scale.X {type=MovieSceneScriptingFloatChannel, default=1}
       thumb_01_r_ctrl.Scale.Y {type=MovieSceneScriptingFloatChannel, default=1}
       thumb_01_r_ctrl.Scale.Z {type=MovieSceneScriptingFloatChannel, default=1}
       thumb_02_r_ctrl.Location.X {type=MovieSceneScriptingFloatChannel, default=0}
       thumb_02_r_ctrl.Location.Y {type=MovieSceneScriptingFloatChannel, default=0}
       thumb_02_r_ctrl.Location.Z {type=MovieSceneScriptingFloatChannel, default=0}
       thumb_02_r_ctrl.Rotation.X {type=MovieSceneScriptingFloatChannel, default=0}
       thumb_02_r_ctrl.Rotation.Y {type=MovieSceneScriptingFloatChannel, default=0}
       thumb_02_r_ctrl.Rotation.Z {type=MovieSceneScriptingFloatChannel, default=0}
       thumb_02_r_ctrl.Scale.X {type=MovieSceneScriptingFloatChannel, default=1}
       thumb_02_r_ctrl.Scale.Y {type=MovieSceneScriptingFloatChannel, default=1}
       thumb_02_r_ctrl.Scale.Z {type=MovieSceneScriptingFloatChannel, default=1}
       thumb_03_r_ctrl.Location.X {type=MovieSceneScriptingFloatChannel, default=0}
       thumb_03_r_ctrl.Location.Y {type=MovieSceneScriptingFloatChannel, default=0}
       thumb_03_r_ctrl.Location.Z {type=MovieSceneScriptingFloatChannel, default=0}
       thumb_03_r_ctrl.Rotation.X {type=MovieSceneScriptingFloatChannel, default=0}
       thumb_03_r_ctrl.Rotation.Y {type=MovieSceneScriptingFloatChannel, default=0}
       thumb_03_r_ctrl.Rotation.Z {type=MovieSceneScriptingFloatChannel, default=0}
       thumb_03_r_ctrl.Scale.X {type=MovieSceneScriptingFloatChannel, default=1}
       thumb_03_r_ctrl.Scale.Y {type=MovieSceneScriptingFloatChannel, default=1}
       thumb_03_r_ctrl.Scale.Z {type=MovieSceneScriptingFloatChannel, default=1}
       index_metacarpal_r_ctrl.Location.X {type=MovieSceneScriptingFloatChannel, default=0}
       index_metacarpal_r_ctrl.Location.Y {type=MovieSceneScriptingFloatChannel, default=0}
       index_metacarpal_r_ctrl.Location.Z {type=MovieSceneScriptingFloatChannel, default=0}
       index_metacarpal_r_ctrl.Rotation.X {type=MovieSceneScriptingFloatChannel, default=0}
       index_metacarpal_r_ctrl.Rotation.Y {type=MovieSceneScriptingFloatChannel, default=0}
       index_metacarpal_r_ctrl.Rotation.Z {type=MovieSceneScriptingFloatChannel, default=0}
       index_metacarpal_r_ctrl.Scale.X {type=MovieSceneScriptingFloatChannel, default=1}
       index_metacarpal_r_ctrl.Scale.Y {type=MovieSceneScriptingFloatChannel, default=1}
       index_metacarpal_r_ctrl.Scale.Z {type=MovieSceneScriptingFloatChannel, default=1}
       index_01_r_ctrl.Location.X {type=MovieSceneScriptingFloatChannel, default=0}
       index_01_r_ctrl.Location.Y {type=MovieSceneScriptingFloatChannel, default=0}
       index_01_r_ctrl.Location.Z {type=MovieSceneScriptingFloatChannel, default=0}
       index_01_r_ctrl.Rotation.X {type=MovieSceneScriptingFloatChannel, default=0}
       index_01_r_ctrl.Rotation.Y {type=MovieSceneScriptingFloatChannel, default=0}
       index_01_r_ctrl.Rotation.Z {type=MovieSceneScriptingFloatChannel, default=0}
       index_01_r_ctrl.Scale.X {type=MovieSceneScriptingFloatChannel, default=1}
       index_01_r_ctrl.Scale.Y {type=MovieSceneScriptingFloatChannel, default=1}
       index_01_r_ctrl.Scale.Z {type=MovieSceneScriptingFloatChannel, default=1}
       index_02_r_ctrl.Location.X {type=MovieSceneScriptingFloatChannel, default=0}
       index_02_r_ctrl.Location.Y {type=MovieSceneScriptingFloatChannel, default=0}
       index_02_r_ctrl.Location.Z {type=MovieSceneScriptingFloatChannel, default=0}
       index_02_r_ctrl.Rotation.X {type=MovieSceneScriptingFloatChannel, default=0}
       index_02_r_ctrl.Rotation.Y {type=MovieSceneScriptingFloatChannel, default=0}
       index_02_r_ctrl.Rotation.Z {type=MovieSceneScriptingFloatChannel, default=0}
       index_02_r_ctrl.Scale.X {type=MovieSceneScriptingFloatChannel, default=1}
       index_02_r_ctrl.Scale.Y {type=MovieSceneScriptingFloatChannel, default=1}
       index_02_r_ctrl.Scale.Z {type=MovieSceneScriptingFloatChannel, default=1}
       index_03_r_ctrl.Location.X {type=MovieSceneScriptingFloatChannel, default=0}
       index_03_r_ctrl.Location.Y {type=MovieSceneScriptingFloatChannel, default=0}
       index_03_r_ctrl.Location.Z {type=MovieSceneScriptingFloatChannel, default=0}
       index_03_r_ctrl.Rotation.X {type=MovieSceneScriptingFloatChannel, default=0}
       index_03_r_ctrl.Rotation.Y {type=MovieSceneScriptingFloatChannel, default=0}
       index_03_r_ctrl.Rotation.Z {type=MovieSceneScriptingFloatChannel, default=0}
       index_03_r_ctrl.Scale.X {type=MovieSceneScriptingFloatChannel, default=1}
       index_03_r_ctrl.Scale.Y {type=MovieSceneScriptingFloatChannel, default=1}
       index_03_r_ctrl.Scale.Z {type=MovieSceneScriptingFloatChannel, default=1}
       middle_metacarpal_r_ctrl.Location.X {type=MovieSceneScriptingFloatChannel, default=0}
       middle_metacarpal_r_ctrl.Location.Y {type=MovieSceneScriptingFloatChannel, default=0}
       middle_metacarpal_r_ctrl.Location.Z {type=MovieSceneScriptingFloatChannel, default=0}
       middle_metacarpal_r_ctrl.Rotation.X {type=MovieSceneScriptingFloatChannel, default=0}
       middle_metacarpal_r_ctrl.Rotation.Y {type=MovieSceneScriptingFloatChannel, default=0}
       middle_metacarpal_r_ctrl.Rotation.Z {type=MovieSceneScriptingFloatChannel, default=0}
       middle_metacarpal_r_ctrl.Scale.X {type=MovieSceneScriptingFloatChannel, default=1}
       middle_metacarpal_r_ctrl.Scale.Y {type=MovieSceneScriptingFloatChannel, default=1}
       middle_metacarpal_r_ctrl.Scale.Z {type=MovieSceneScriptingFloatChannel, default=1}
       middle_01_r_ctrl.Location.X {type=MovieSceneScriptingFloatChannel, default=0}
       middle_01_r_ctrl.Location.Y {type=MovieSceneScriptingFloatChannel, default=0}
       middle_01_r_ctrl.Location.Z {type=MovieSceneScriptingFloatChannel, default=0}
       middle_01_r_ctrl.Rotation.X {type=MovieSceneScriptingFloatChannel, default=0}
       middle_01_r_ctrl.Rotation.Y {type=MovieSceneScriptingFloatChannel, default=0}
       middle_01_r_ctrl.Rotation.Z {type=MovieSceneScriptingFloatChannel, default=0}
       middle_01_r_ctrl.Scale.X {type=MovieSceneScriptingFloatChannel, default=1}
       middle_01_r_ctrl.Scale.Y {type=MovieSceneScriptingFloatChannel, default=1}
       middle_01_r_ctrl.Scale.Z {type=MovieSceneScriptingFloatChannel, default=1}
       middle_02_r_ctrl.Location.X {type=MovieSceneScriptingFloatChannel, default=0}
       middle_02_r_ctrl.Location.Y {type=MovieSceneScriptingFloatChannel, default=0}
       middle_02_r_ctrl.Location.Z {type=MovieSceneScriptingFloatChannel, default=0}
       middle_02_r_ctrl.Rotation.X {type=MovieSceneScriptingFloatChannel, default=0}
       middle_02_r_ctrl.Rotation.Y {type=MovieSceneScriptingFloatChannel, default=0}
       middle_02_r_ctrl.Rotation.Z {type=MovieSceneScriptingFloatChannel, default=0}
       middle_02_r_ctrl.Scale.X {type=MovieSceneScriptingFloatChannel, default=1}
       middle_02_r_ctrl.Scale.Y {type=MovieSceneScriptingFloatChannel, default=1}
       middle_02_r_ctrl.Scale.Z {type=MovieSceneScriptingFloatChannel, default=1}
       middle_03_r_ctrl.Location.X {type=MovieSceneScriptingFloatChannel, default=0}
       middle_03_r_ctrl.Location.Y {type=MovieSceneScriptingFloatChannel, default=0}
       middle_03_r_ctrl.Location.Z {type=MovieSceneScriptingFloatChannel, default=0}
       middle_03_r_ctrl.Rotation.X {type=MovieSceneScriptingFloatChannel, default=0}
       middle_03_r_ctrl.Rotation.Y {type=MovieSceneScriptingFloatChannel, default=0}
       middle_03_r_ctrl.Rotation.Z {type=MovieSceneScriptingFloatChannel, default=0}
       middle_03_r_ctrl.Scale.X {type=MovieSceneScriptingFloatChannel, default=1}
       middle_03_r_ctrl.Scale.Y {type=MovieSceneScriptingFloatChannel, default=1}
       middle_03_r_ctrl.Scale.Z {type=MovieSceneScriptingFloatChannel, default=1}
       ring_metacarpal_r_ctrl.Location.X {type=MovieSceneScriptingFloatChannel, default=0}
       ring_metacarpal_r_ctrl.Location.Y {type=MovieSceneScriptingFloatChannel, default=0}
       ring_metacarpal_r_ctrl.Location.Z {type=MovieSceneScriptingFloatChannel, default=0}
       ring_metacarpal_r_ctrl.Rotation.X {type=MovieSceneScriptingFloatChannel, default=0}
       ring_metacarpal_r_ctrl.Rotation.Y {type=MovieSceneScriptingFloatChannel, default=0}
       ring_metacarpal_r_ctrl.Rotation.Z {type=MovieSceneScriptingFloatChannel, default=0}
       ring_metacarpal_r_ctrl.Scale.X {type=MovieSceneScriptingFloatChannel, default=1}
       ring_metacarpal_r_ctrl.Scale.Y {type=MovieSceneScriptingFloatChannel, default=1}
       ring_metacarpal_r_ctrl.Scale.Z {type=MovieSceneScriptingFloatChannel, default=1}
       ring_01_r_ctrl.Location.X {type=MovieSceneScriptingFloatChannel, default=0}
       ring_01_r_ctrl.Location.Y {type=MovieSceneScriptingFloatChannel, default=0}
       ring_01_r_ctrl.Location.Z {type=MovieSceneScriptingFloatChannel, default=0}
       ring_01_r_ctrl.Rotation.X {type=MovieSceneScriptingFloatChannel, default=0}
       ring_01_r_ctrl.Rotation.Y {type=MovieSceneScriptingFloatChannel, default=0}
       ring_01_r_ctrl.Rotation.Z {type=MovieSceneScriptingFloatChannel, default=0}
       ring_01_r_ctrl.Scale.X {type=MovieSceneScriptingFloatChannel, default=1}
       ring_01_r_ctrl.Scale.Y {type=MovieSceneScriptingFloatChannel, default=1}
       ring_01_r_ctrl.Scale.Z {type=MovieSceneScriptingFloatChannel, default=1}
       ring_02_r_ctrl.Location.X {type=MovieSceneScriptingFloatChannel, default=0}
       ring_02_r_ctrl.Location.Y {type=MovieSceneScriptingFloatChannel, default=0}
       ring_02_r_ctrl.Location.Z {type=MovieSceneScriptingFloatChannel, default=0}
       ring_02_r_ctrl.Rotation.X {type=MovieSceneScriptingFloatChannel, default=0}
       ring_02_r_ctrl.Rotation.Y {type=MovieSceneScriptingFloatChannel, default=0}
       ring_02_r_ctrl.Rotation.Z {type=MovieSceneScriptingFloatChannel, default=0}
       ring_02_r_ctrl.Scale.X {type=MovieSceneScriptingFloatChannel, default=1}
       ring_02_r_ctrl.Scale.Y {type=MovieSceneScriptingFloatChannel, default=1}
       ring_02_r_ctrl.Scale.Z {type=MovieSceneScriptingFloatChannel, default=1}
       ring_03_r_ctrl.Location.X {type=MovieSceneScriptingFloatChannel, default=0}
       ring_03_r_ctrl.Location.Y {type=MovieSceneScriptingFloatChannel, default=0}
       ring_03_r_ctrl.Location.Z {type=MovieSceneScriptingFloatChannel, default=0}
       ring_03_r_ctrl.Rotation.X {type=MovieSceneScriptingFloatChannel, default=0}
       ring_03_r_ctrl.Rotation.Y {type=MovieSceneScriptingFloatChannel, default=0}
       ring_03_r_ctrl.Rotation.Z {type=MovieSceneScriptingFloatChannel, default=0}
       ring_03_r_ctrl.Scale.X {type=MovieSceneScriptingFloatChannel, default=1}
       ring_03_r_ctrl.Scale.Y {type=MovieSceneScriptingFloatChannel, default=1}
       ring_03_r_ctrl.Scale.Z {type=MovieSceneScriptingFloatChannel, default=1}
       pinky_metacarpal_r_ctrl.Location.X {type=MovieSceneScriptingFloatChannel, default=0}
       pinky_metacarpal_r_ctrl.Location.Y {type=MovieSceneScriptingFloatChannel, default=0}
       pinky_metacarpal_r_ctrl.Location.Z {type=MovieSceneScriptingFloatChannel, default=0}
       pinky_metacarpal_r_ctrl.Rotation.X {type=MovieSceneScriptingFloatChannel, default=0}
       pinky_metacarpal_r_ctrl.Rotation.Y {type=MovieSceneScriptingFloatChannel, default=0}
       pinky_metacarpal_r_ctrl.Rotation.Z {type=MovieSceneScriptingFloatChannel, default=0}
       pinky_metacarpal_r_ctrl.Scale.X {type=MovieSceneScriptingFloatChannel, default=1}
       pinky_metacarpal_r_ctrl.Scale.Y {type=MovieSceneScriptingFloatChannel, default=1}
       pinky_metacarpal_r_ctrl.Scale.Z {type=MovieSceneScriptingFloatChannel, default=1}
       pinky_01_r_ctrl.Location.X {type=MovieSceneScriptingFloatChannel, default=0}
       pinky_01_r_ctrl.Location.Y {type=MovieSceneScriptingFloatChannel, default=0}
       pinky_01_r_ctrl.Location.Z {type=MovieSceneScriptingFloatChannel, default=0}
       pinky_01_r_ctrl.Rotation.X {type=MovieSceneScriptingFloatChannel, default=0}
       pinky_01_r_ctrl.Rotation.Y {type=MovieSceneScriptingFloatChannel, default=0}
       pinky_01_r_ctrl.Rotation.Z {type=MovieSceneScriptingFloatChannel, default=0}
       pinky_01_r_ctrl.Scale.X {type=MovieSceneScriptingFloatChannel, default=1}
       pinky_01_r_ctrl.Scale.Y {type=MovieSceneScriptingFloatChannel, default=1}
       pinky_01_r_ctrl.Scale.Z {type=MovieSceneScriptingFloatChannel, default=1}
       pinky_02_r_ctrl.Location.X {type=MovieSceneScriptingFloatChannel, default=0}
       pinky_02_r_ctrl.Location.Y {type=MovieSceneScriptingFloatChannel, default=0}
       pinky_02_r_ctrl.Location.Z {type=MovieSceneScriptingFloatChannel, default=0}
       pinky_02_r_ctrl.Rotation.X {type=MovieSceneScriptingFloatChannel, default=0}
       pinky_02_r_ctrl.Rotation.Y {type=MovieSceneScriptingFloatChannel, default=0}
       pinky_02_r_ctrl.Rotation.Z {type=MovieSceneScriptingFloatChannel, default=0}
       pinky_02_r_ctrl.Scale.X {type=MovieSceneScriptingFloatChannel, default=1}
       pinky_02_r_ctrl.Scale.Y {type=MovieSceneScriptingFloatChannel, default=1}
       pinky_02_r_ctrl.Scale.Z {type=MovieSceneScriptingFloatChannel, default=1}
       pinky_03_r_ctrl.Location.X {type=MovieSceneScriptingFloatChannel, default=0}
       pinky_03_r_ctrl.Location.Y {type=MovieSceneScriptingFloatChannel, default=0}
       pinky_03_r_ctrl.Location.Z {type=MovieSceneScriptingFloatChannel, default=0}
       pinky_03_r_ctrl.Rotation.X {type=MovieSceneScriptingFloatChannel, default=0}
       pinky_03_r_ctrl.Rotation.Y {type=MovieSceneScriptingFloatChannel, default=0}
       pinky_03_r_ctrl.Rotation.Z {type=MovieSceneScriptingFloatChannel, default=0}
       pinky_03_r_ctrl.Scale.X {type=MovieSceneScriptingFloatChannel, default=1}
       pinky_03_r_ctrl.Scale.Y {type=MovieSceneScriptingFloatChannel, default=1}
       pinky_03_r_ctrl.Scale.Z {type=MovieSceneScriptingFloatChannel, default=1}
       hand_r_ik_ctrl.Location.X {type=MovieSceneScriptingFloatChannel, default=0}
       hand_r_ik_ctrl.Location.Y {type=MovieSceneScriptingFloatChannel, default=0}
       hand_r_ik_ctrl.Location.Z {type=MovieSceneScriptingFloatChannel, default=0}
       hand_r_ik_ctrl.Rotation.X {type=MovieSceneScriptingFloatChannel, default=0}
       hand_r_ik_ctrl.Rotation.Y {type=MovieSceneScriptingFloatChannel, default=0}
       hand_r_ik_ctrl.Rotation.Z {type=MovieSceneScriptingFloatChannel, default=0}
       hand_r_ik_ctrl.Scale.X {type=MovieSceneScriptingFloatChannel, default=1}
       hand_r_ik_ctrl.Scale.Y {type=MovieSceneScriptingFloatChannel, default=1}
       hand_r_ik_ctrl.Scale.Z {type=MovieSceneScriptingFloatChannel, default=1}
       Softness_2 {type=MovieSceneScriptingFloatChannel, default=0}
       hand_r_ikorient_ctrl.Rotation.Roll {type=MovieSceneScriptingFloatChannel, default=0}
       hand_r_ikorient_ctrl.Rotation.Pitch {type=MovieSceneScriptingFloatChannel, default=0}
       hand_r_ikorient_ctrl.Rotation.Yaw {type=MovieSceneScriptingFloatChannel, default=0}
       arm_r_pv_ik_ctrl.Location.X {type=MovieSceneScriptingFloatChannel, default=-0.6565}
       arm_r_pv_ik_ctrl.Location.Y {type=MovieSceneScriptingFloatChannel, default=-3.3577}
       arm_r_pv_ik_ctrl.Location.Z {type=MovieSceneScriptingFloatChannel, default=2.1601}
       arm_r_pv_ik_ctrl.Rotation.X {type=MovieSceneScriptingFloatChannel, default=6.0596}
       arm_r_pv_ik_ctrl.Rotation.Y {type=MovieSceneScriptingFloatChannel, default=-1.7562}
       arm_r_pv_ik_ctrl.Rotation.Z {type=MovieSceneScriptingFloatChannel, default=0.7302}
       arm_r_pv_ik_ctrl.Scale.X {type=MovieSceneScriptingFloatChannel, default=1}
       arm_r_pv_ik_ctrl.Scale.Y {type=MovieSceneScriptingFloatChannel, default=1}
       arm_r_pv_ik_ctrl.Scale.Z {type=MovieSceneScriptingFloatChannel, default=1}
       body_ctrl.Location.X {type=MovieSceneScriptingFloatChannel, default=0}
       body_ctrl.Location.Y {type=MovieSceneScriptingFloatChannel, default=0}
       body_ctrl.Location.Z {type=MovieSceneScriptingFloatChannel, default=0}
       body_ctrl.Rotation.X {type=MovieSceneScriptingFloatChannel, default=0}
       body_ctrl.Rotation.Y {type=MovieSceneScriptingFloatChannel, default=0}
       body_ctrl.Rotation.Z {type=MovieSceneScriptingFloatChannel, default=0}
       body_ctrl.Scale.X {type=MovieSceneScriptingFloatChannel, default=1}
       body_ctrl.Scale.Y {type=MovieSceneScriptingFloatChannel, default=1}
       body_ctrl.Scale.Z {type=MovieSceneScriptingFloatChannel, default=1}
       hips_ctrl.Location.X {type=MovieSceneScriptingFloatChannel, default=0}
       hips_ctrl.Location.Y {type=MovieSceneScriptingFloatChannel, default=0}
       hips_ctrl.Location.Z {type=MovieSceneScriptingFloatChannel, default=0}
       hips_ctrl.Rotation.X {type=MovieSceneScriptingFloatChannel, default=0}
       hips_ctrl.Rotation.Y {type=MovieSceneScriptingFloatChannel, default=0}
       hips_ctrl.Rotation.Z {type=MovieSceneScriptingFloatChannel, default=0}
       hips_ctrl.Scale.X {type=MovieSceneScriptingFloatChannel, default=1}
       hips_ctrl.Scale.Y {type=MovieSceneScriptingFloatChannel, default=1}
       hips_ctrl.Scale.Z {type=MovieSceneScriptingFloatChannel, default=1}
       hips_tan_ctrl.Location.X {type=MovieSceneScriptingFloatChannel, default=0}
       hips_tan_ctrl.Location.Y {type=MovieSceneScriptingFloatChannel, default=0}
       hips_tan_ctrl.Location.Z {type=MovieSceneScriptingFloatChannel, default=0}
       hips_tan_ctrl.Rotation.X {type=MovieSceneScriptingFloatChannel, default=0}
       hips_tan_ctrl.Rotation.Y {type=MovieSceneScriptingFloatChannel, default=0}
       hips_tan_ctrl.Rotation.Z {type=MovieSceneScriptingFloatChannel, default=0}
       hips_tan_ctrl.Scale.X {type=MovieSceneScriptingFloatChannel, default=1}
       hips_tan_ctrl.Scale.Y {type=MovieSceneScriptingFloatChannel, default=1}
       hips_tan_ctrl.Scale.Z {type=MovieSceneScriptingFloatChannel, default=1}
       spine_01_ctrl.Location.X {type=MovieSceneScriptingFloatChannel, default=0}
       spine_01_ctrl.Location.Y {type=MovieSceneScriptingFloatChannel, default=0}
       spine_01_ctrl.Location.Z {type=MovieSceneScriptingFloatChannel, default=0}
       spine_01_ctrl.Rotation.X {type=MovieSceneScriptingFloatChannel, default=0}
       spine_01_ctrl.Rotation.Y {type=MovieSceneScriptingFloatChannel, default=0}
       spine_01_ctrl.Rotation.Z {type=MovieSceneScriptingFloatChannel, default=0}
       spine_01_ctrl.Scale.X {type=MovieSceneScriptingFloatChannel, default=1}
       spine_01_ctrl.Scale.Y {type=MovieSceneScriptingFloatChannel, default=1}
       spine_01_ctrl.Scale.Z {type=MovieSceneScriptingFloatChannel, default=1}
       spine_02_ctrl.Location.X {type=MovieSceneScriptingFloatChannel, default=0}
       spine_02_ctrl.Location.Y {type=MovieSceneScriptingFloatChannel, default=0}
       spine_02_ctrl.Location.Z {type=MovieSceneScriptingFloatChannel, default=0}
       spine_02_ctrl.Rotation.X {type=MovieSceneScriptingFloatChannel, default=0}
       spine_02_ctrl.Rotation.Y {type=MovieSceneScriptingFloatChannel, default=0}
       spine_02_ctrl.Rotation.Z {type=MovieSceneScriptingFloatChannel, default=0}
       spine_02_ctrl.Scale.X {type=MovieSceneScriptingFloatChannel, default=1}
       spine_02_ctrl.Scale.Y {type=MovieSceneScriptingFloatChannel, default=1}
       spine_02_ctrl.Scale.Z {type=MovieSceneScriptingFloatChannel, default=1}
       spine_03_ctrl.Location.X {type=MovieSceneScriptingFloatChannel, default=0}
       spine_03_ctrl.Location.Y {type=MovieSceneScriptingFloatChannel, default=0}
       spine_03_ctrl.Location.Z {type=MovieSceneScriptingFloatChannel, default=0}
       spine_03_ctrl.Rotation.X {type=MovieSceneScriptingFloatChannel, default=0}
       spine_03_ctrl.Rotation.Y {type=MovieSceneScriptingFloatChannel, default=0}
       spine_03_ctrl.Rotation.Z {type=MovieSceneScriptingFloatChannel, default=0}
       spine_03_ctrl.Scale.X {type=MovieSceneScriptingFloatChannel, default=1}
       spine_03_ctrl.Scale.Y {type=MovieSceneScriptingFloatChannel, default=1}
       spine_03_ctrl.Scale.Z {type=MovieSceneScriptingFloatChannel, default=1}
       spine_04_ctrl.Location.X {type=MovieSceneScriptingFloatChannel, default=0}
       spine_04_ctrl.Location.Y {type=MovieSceneScriptingFloatChannel, default=0}
       spine_04_ctrl.Location.Z {type=MovieSceneScriptingFloatChannel, default=0}
       spine_04_ctrl.Rotation.X {type=MovieSceneScriptingFloatChannel, default=0}
       spine_04_ctrl.Rotation.Y {type=MovieSceneScriptingFloatChannel, default=0}
       spine_04_ctrl.Rotation.Z {type=MovieSceneScriptingFloatChannel, default=0}
       spine_04_ctrl.Scale.X {type=MovieSceneScriptingFloatChannel, default=1}
       spine_04_ctrl.Scale.Y {type=MovieSceneScriptingFloatChannel, default=1}
       spine_04_ctrl.Scale.Z {type=MovieSceneScriptingFloatChannel, default=1}
       head_ctrl.Location.X {type=MovieSceneScriptingFloatChannel, default=0}
       head_ctrl.Location.Y {type=MovieSceneScriptingFloatChannel, default=0}
       head_ctrl.Location.Z {type=MovieSceneScriptingFloatChannel, default=0}
       head_ctrl.Rotation.X {type=MovieSceneScriptingFloatChannel, default=0}
       head_ctrl.Rotation.Y {type=MovieSceneScriptingFloatChannel, default=0}
       head_ctrl.Rotation.Z {type=MovieSceneScriptingFloatChannel, default=0}
       head_ctrl.Scale.X {type=MovieSceneScriptingFloatChannel, default=1}
       head_ctrl.Scale.Y {type=MovieSceneScriptingFloatChannel, default=1}
       head_ctrl.Scale.Z {type=MovieSceneScriptingFloatChannel, default=1}
       clavicle_l_ctrl.Location.X {type=MovieSceneScriptingFloatChannel, default=0}
       clavicle_l_ctrl.Location.Y {type=MovieSceneScriptingFloatChannel, default=0}
       clavicle_l_ctrl.Location.Z {type=MovieSceneScriptingFloatChannel, default=0}
       clavicle_l_ctrl.Rotation.X {type=MovieSceneScriptingFloatChannel, default=0}
       clavicle_l_ctrl.Rotation.Y {type=MovieSceneScriptingFloatChannel, default=0}
       clavicle_l_ctrl.Rotation.Z {type=MovieSceneScriptingFloatChannel, default=0}
       clavicle_l_ctrl.Scale.X {type=MovieSceneScriptingFloatChannel, default=1}
       clavicle_l_ctrl.Scale.Y {type=MovieSceneScriptingFloatChannel, default=1}
       clavicle_l_ctrl.Scale.Z {type=MovieSceneScriptingFloatChannel, default=1}
       clavicle_r_ctrl.Location.X {type=MovieSceneScriptingFloatChannel, default=0}
       clavicle_r_ctrl.Location.Y {type=MovieSceneScriptingFloatChannel, default=0}
       clavicle_r_ctrl.Location.Z {type=MovieSceneScriptingFloatChannel, default=0}
       clavicle_r_ctrl.Rotation.X {type=MovieSceneScriptingFloatChannel, default=0}
       clavicle_r_ctrl.Rotation.Y {type=MovieSceneScriptingFloatChannel, default=0}
       clavicle_r_ctrl.Rotation.Z {type=MovieSceneScriptingFloatChannel, default=0}
       clavicle_r_ctrl.Scale.X {type=MovieSceneScriptingFloatChannel, default=1}
       clavicle_r_ctrl.Scale.Y {type=MovieSceneScriptingFloatChannel, default=1}
       clavicle_r_ctrl.Scale.Z {type=MovieSceneScriptingFloatChannel, default=1}
       neck_01_ctrl.Location.X {type=MovieSceneScriptingFloatChannel, default=0}
       neck_01_ctrl.Location.Y {type=MovieSceneScriptingFloatChannel, default=0}
       neck_01_ctrl.Location.Z {type=MovieSceneScriptingFloatChannel, default=0}
       neck_01_ctrl.Rotation.X {type=MovieSceneScriptingFloatChannel, default=0}
       neck_01_ctrl.Rotation.Y {type=MovieSceneScriptingFloatChannel, default=0}
       neck_01_ctrl.Rotation.Z {type=MovieSceneScriptingFloatChannel, default=0}
       neck_01_ctrl.Scale.X {type=MovieSceneScriptingFloatChannel, default=1}
       neck_01_ctrl.Scale.Y {type=MovieSceneScriptingFloatChannel, default=1}
       neck_01_ctrl.Scale.Z {type=MovieSceneScriptingFloatChannel, default=1}
       neck_02_ctrl.Location.X {type=MovieSceneScriptingFloatChannel, default=0}
       neck_02_ctrl.Location.Y {type=MovieSceneScriptingFloatChannel, default=0}
       neck_02_ctrl.Location.Z {type=MovieSceneScriptingFloatChannel, default=0}
       neck_02_ctrl.Rotation.X {type=MovieSceneScriptingFloatChannel, default=0}
       neck_02_ctrl.Rotation.Y {type=MovieSceneScriptingFloatChannel, default=0}
       neck_02_ctrl.Rotation.Z {type=MovieSceneScriptingFloatChannel, default=0}
       neck_02_ctrl.Scale.X {type=MovieSceneScriptingFloatChannel, default=1}
       neck_02_ctrl.Scale.Y {type=MovieSceneScriptingFloatChannel, default=1}
       neck_02_ctrl.Scale.Z {type=MovieSceneScriptingFloatChannel, default=1}
       head_ik_ctrl.Location.X {type=MovieSceneScriptingFloatChannel, default=0}
       head_ik_ctrl.Location.Y {type=MovieSceneScriptingFloatChannel, default=0}
       head_ik_ctrl.Location.Z {type=MovieSceneScriptingFloatChannel, default=0}
       head_ik_ctrl.Rotation.X {type=MovieSceneScriptingFloatChannel, default=0}
       head_ik_ctrl.Rotation.Y {type=MovieSceneScriptingFloatChannel, default=0}
       head_ik_ctrl.Rotation.Z {type=MovieSceneScriptingFloatChannel, default=0}
       head_ik_ctrl.Scale.X {type=MovieSceneScriptingFloatChannel, default=1}
       head_ik_ctrl.Scale.Y {type=MovieSceneScriptingFloatChannel, default=1}
       head_ik_ctrl.Scale.Z {type=MovieSceneScriptingFloatChannel, default=1}
       chest_ctrl.Location.X {type=MovieSceneScriptingFloatChannel, default=0}
       chest_ctrl.Location.Y {type=MovieSceneScriptingFloatChannel, default=0}
       chest_ctrl.Location.Z {type=MovieSceneScriptingFloatChannel, default=0}
       chest_ctrl.Rotation.X {type=MovieSceneScriptingFloatChannel, default=0}
       chest_ctrl.Rotation.Y {type=MovieSceneScriptingFloatChannel, default=0}
       chest_ctrl.Rotation.Z {type=MovieSceneScriptingFloatChannel, default=0}
       chest_ctrl.Scale.X {type=MovieSceneScriptingFloatChannel, default=1}
       chest_ctrl.Scale.Y {type=MovieSceneScriptingFloatChannel, default=1}
       chest_ctrl.Scale.Z {type=MovieSceneScriptingFloatChannel, default=1}
       chest_tan_ctrl.Location.X {type=MovieSceneScriptingFloatChannel, default=0}
       chest_tan_ctrl.Location.Y {type=MovieSceneScriptingFloatChannel, default=0}
       chest_tan_ctrl.Location.Z {type=MovieSceneScriptingFloatChannel, default=0}
       chest_tan_ctrl.Rotation.X {type=MovieSceneScriptingFloatChannel, default=0}
       chest_tan_ctrl.Rotation.Y {type=MovieSceneScriptingFloatChannel, default=0}
       chest_tan_ctrl.Rotation.Z {type=MovieSceneScriptingFloatChannel, default=0}
       chest_tan_ctrl.Scale.X {type=MovieSceneScriptingFloatChannel, default=1}
       chest_tan_ctrl.Scale.Y {type=MovieSceneScriptingFloatChannel, default=1}
       chest_tan_ctrl.Scale.Z {type=MovieSceneScriptingFloatChannel, default=1}
       foot_r_ik_ctrl.Space {type=MovieSceneScriptingSpaceChannel}
      ]
     }]
   }
  ]
 }]
display_rate {denominator=1, numerator=30}
tick_resolution {denominator=1, numerator=24000}
tracks []
